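# Looping ambient sound cue for a spatialized barbecue sizzle prop.
# One looping wave gets random pitch modulation and two nested endless envelopes of different lengths that slowly drift pitch and volume, masking repetition.
# A random node starts it immediately or after a short random delay, under a brief root fade-in.
# Overrides set capsule attenuation with occlusion and listener focus, per-owner concurrency, and reverb and exterior-ambience sends.
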SoundCue /Game/Sounds/Helios/Environment_Actors/BBQSizzle/Spot_BBQ_Sizzle_Loop_Cue
root=Enveloper_1
nodes [
 {id=Delay_0, children [Enveloper_3], properties {delayMax=0.5, delayMin=0.2}}
 {
  id=Enveloper_0
  children [Modulator_0]
  properties {
   bLoop=true
   bLoopIndefinitely=true
   loopEnd=30
   pitchCurve {editorCurveData {keys [
      {arriveTangent=-0.0013, interpMode=RCIM_Cubic, leaveTangent=-0.0013, time=0}
      <at 15 0.98 RCIM_Cubic>
      {arriveTangent=0.0013, interpMode=RCIM_Cubic, leaveTangent=0.0013, time=30}
     ]}}
   volumeCurve {editorCurveData {keys [<at 0 1>]}}
  }
 }
 {
  id=Enveloper_1
  children [Random_2]
  properties {pitchCurve {editorCurveData {keys [{time=0}]}}, volumeCurve {editorCurveData {keys [<at 0 0>,<at 0.2 1>]}}}
 }
 {
  id=Enveloper_3
  children [Enveloper_0]
  properties {
   bLoop=true
   bLoopIndefinitely=true
   loopEnd=20
   pitchCurve {editorCurveData {keys [{time=0}]}}
   volumeCurve {editorCurveData {keys [
      {arriveTangent=-0.01, interpMode=RCIM_Cubic, leaveTangent=-0.01, time=0, value=1}
      <at 10 0.9 RCIM_Cubic>
      {arriveTangent=0.01, interpMode=RCIM_Cubic, leaveTangent=0.01, time=20, value=1}
     ]}}
  }
 }
 {id=Modulator_0, children [WavePlayer_1], properties {pitchMax=1.1, pitchMin=0.9}}
 {id=Random_2, children [Delay_0,Enveloper_3], properties {weights [1,1]}}
 {id=WavePlayer_1, properties {bLooping=true, soundWaveAssetPtr=&/Spot_Barbecue_Sizzle_Loop}}
]
properties {
 bOverrideAttenuation=true
 bOverrideConcurrency=true
 maxDistance=684
 soundClassObject=&Fort_Ambient_Props_Athena
 subtitlePriority=5000
 volumeMultiplier=0.301
 attenuationOverrides {
  attenuationShape=Capsule
  bEnableFocusInterpolation=true
  bEnableListenerFocus=true
  bEnableOcclusion=true
  bEnableSubmixSends=true
  dBAttenuationAtMax=-40
  distanceAlgorithm=Custom
  falloffDistance=512
  nonFocusVolumeAttenuation=0.1
  occlusionInterpolationTime=1
  occlusionLowPassFilterFrequency=10000
  occlusionTraceChannel=ECC_GameTraceChannel11
  occlusionVolumeAttenuation=0.5
  stereoSpread=100
  attenuationShapeExtents=<172 88 90>
  customAttenuationCurve {editorCurveData {keys [
     <at 0 1>
     {arriveTangent=-1.0243, interpMode=RCIM_Cubic, leaveTangent=-1.0243, tangentMode=RCTM_User, time=0.46, value=0.3651}
     {arriveTangent=-1.8058, interpMode=RCIM_Cubic, leaveTangent=-1.8058, tangentMode=RCTM_User, time=1, value=0}
    ]}}
 }
 busSends [{
   audioBus=null
   minSendLevel=0.6
   sendLevel=1
   soundSourceBus=&AmbientAudio_Exterior_SourceBus
   customSendLevelCurve {editorCurveData {keys [<at 0 0.2>,<at 1 1>]}}
  }]
 concurrencyOverrides {bLimitToOwner=true, maxCount=3}
 soundSubmixSends [{
   minSendDistance=172
   minSendLevel=0.6
   soundSubmix=&Fortnite_Reverb_Default
   customSendLevelCurve {editorCurveData {keys [<at 0 0.2>,<at 1 1>]}}
  }]
}
